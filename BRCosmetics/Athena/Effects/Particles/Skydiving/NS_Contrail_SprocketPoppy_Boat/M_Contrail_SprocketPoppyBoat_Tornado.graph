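# Translucent unlit Niagara mesh-particle material for the tornado in the SprocketPoppy Boat skydiving contrail.
# Emissive sums two tinted, disturbance-warped panning noises, each sampled at three UV offsets for RGB fringing, modulated by V- and Fresnel-driven curve-atlas ramps.
# Opacity combines scrolling smoke, end and edge masks, depth, view-angle and near-camera fades, and a particle-driven HeightLerp dissolve.
# Dynamic parameters also drive a clamped, UV-weighted position offset that bends the mesh.
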Material /BRCosmetics/Athena/Effects/Particles/Skydiving/NS_Contrail_SprocketPoppy_Boat/M_Contrail_SprocketPoppyBoat_Tornado
graph {
 properties {
  bAutomaticallySetUsageInEditor=false
  bUsedWithNiagaraMeshParticles=true
  blendMode=BLEND_Translucent
  shadingModel=MSM_Unlit
  usedShadingModels=Unlit
 }
 nodes [
  n1 = DynamicParameter() {parameterIndex=1, paramNames [X,Y,Z,WPOIntensity]}
  n2 = MaterialFunctionCall(X=@n1.X, Y=@n1.X, Z=@n1.X) {materialFunction=&MakeFloat3}
  n3 = ParticlePositionWS()
  n4 = Constant() {r=50}
  n5 = OneMinus(None=ComponentMask(None=TextureCoordinate()) {g=true})
  n6 = ScalarParameter() {defaultValue=3, parameterName=WPOBendPower}
  LinearInterpolate(
   B=Clamp(None=Subtract(A=@n2.Result, B=@n3), Min=Multiply(A=@n4) {constB=-1}, Max=@n4)
   Alpha=Multiply(A=@n1.WPOIntensity, B=Power(Base=@n5, Exp=@n6))
  )
  MaterialFunctionCall() {materialFunction=&DebugTimeSine}
  n7 = MaterialFunctionCall(Distance=ScalarParameter() {defaultValue=200, parameterName=NearCameraFade}) {materialFunction=&Near_Camera_Fade}
  n8 = Constant() {r=0.5}
  n9 = Panner(
   Coordinate=TextureCoordinate()
   Speed=ComponentMask(None=VectorParameter() {parameterName=AlphaSpeed, defaultValue=<rgba 0 1 0 0>}.RGB) {g=true, r=true}
  ) {speedY=1}
  n10 = Multiply(
   A=Constant() {r=1}
   B=Power(
    Base=TextureSampleParameter2D(UVs=@n9) {parameterName=Alpha, texture=&T_Inky_Smoke_Tile}.R
    Exp=ScalarParameter() {defaultValue=0.5, parameterName=AlphaPower}
   ) {constExponent=0.5}
  )
  n11 = TextureCoordinate()
  n12 = ComponentMask(None=@n11) {g=true}
  n13 = ScalarParameter() {defaultValue=1.5, parameterName=TopandDownAlphaPower}
  n14 = Saturate(None=Multiply(A=Power(Base=@n12, Exp=@n13) {constExponent=1.5}))
  n15 = OneMinus(None=ComponentMask(None=@n11) {g=true})
  n16 = Saturate(None=Multiply(A=Power(Base=@n15, Exp=@n13) {constExponent=1.5}))
  n17 = Saturate(None=Multiply(A=Multiply(A=@n14, B=@n16)) {constB=50})
  n18 = OneMinus(None=Fresnel() {baseReflectFraction=0, exponent=1})
  n19 = Sine(None=ComponentMask(None=TextureCoordinate()) {g=true}) {period=2}
  n20 = Multiply(A=Power(Base=OneMinus(None=@n19)) {constExponent=1}) {constB=3}
  n21 = Multiply(
   A=Multiply(A=Multiply(A=@n10, B=@n17), B=Power(Base=@n18) {constExponent=1})
   B=LinearInterpolate(B=Saturate(None=@n20), Alpha=ScalarParameter() {parameterName=UseCenterAlpha}) {constA=1}
  )
  n22 = ScalarParameter() {defaultValue=1, parameterName=MainAlpha}
  n23 = DepthFade(
   Opacity=Multiply(
    A=Multiply(A=@n21, B=@n22)
    B=MaterialFunctionCall("FadeSensitivity (optional)"=@n8) {materialFunction=&FadeBasedOnViewAngle}.Result
   )
   FadeDistance=ScalarParameter() {defaultValue=100, parameterName=FadeDistance}
  ) {fadeDistanceDefault=0}
  n24 = TextureCoordinate()
  n25 = MaterialFunctionCall(
   Transition Phase=DynamicParameter() {defaultValue=<rgba 1 1 1 0>, paramNames [Param1,Param2,Param3,Dissolve]}.Dissolve
   Height Texture=TextureSample(UVs=Panner(Coordinate=@n24) {speedY=1}) {texture=/BRCosmetics/Weapons/FORT_Melee/Pickaxe_CoatCheck/FX/T_Pickaxe_CoatCheck_Noise}.R
   Contrast=ScalarParameter() {defaultValue=0.3, parameterName=DissolveContrast}
  ) {materialFunction=&HeightLerp}
  Multiply(A=Multiply(A=ParticleColor().A, B=Multiply(A=@n7.Result, B=@n23)), B=@n25.Results)
  Multiply(A=@n25.Results)
  n26 = TextureSampleParameter2D(UVs=Panner(Coordinate=TextureCoordinate()) {speedX=-0.5, speedY=0.7}) {parameterName=DisturbTex, texture=&T_Inky_Smoke_Tile}
  NamedRerouteDeclaration(None=Multiply(A=ComponentMask(None=@n26.RGB) {g=true, r=true}, B=ScalarParameter() {defaultValue=0.2, parameterName=DisturbIntensity}) {constB=0.2}) {name=Disturb, nodeColor=<rgba 0 1 0.5313 1>}
  n27 = ComponentMask(None=TextureCoordinate()) {g=true}
  n28 = TextureSample(UVs=Panner(Coordinate=TextureCoordinate()) {speedY=1}) {texture=/Game/Effects/Fort_Effects/Textures/Power_Source/T_Power_Source_Sphere_Texture}
  n29 = CurveAtlasRowParameter(CurveTime=Power(Base=Sine(None=@n27) {period=2}, Exp=LinearInterpolate(Alpha=@n28.R) {constA=3, constB=4}) {constExponent=3}) {
   atlas=&/CAtlas_Contrail_SprocketPoppy
   curve=&/CV_ContrailSprocketPoppy
   parameterName=ColorCV
  }
  n30 = Fresnel() {baseReflectFraction=0, exponent=1}
  n31 = CurveAtlasRowParameter(CurveTime=Power(Base=OneMinus(None=@n30), Exp=LinearInterpolate(Alpha=@n28.R) {constA=3, constB=10}) {constExponent=5}) {
   atlas=&CAtlas_Contrail_SprocketPoppy
   curve=&/CV_ContrailSprocketPoppy1
   defaultValue=1
   parameterName=ColorCV2
  }
  n32 = Multiply(A=@n29, B=@n31)
  n33 = VectorParameter() {parameterName=Color2, defaultValue=<rgba 8.2139 10 5.7465 1>}
  n34 = Add(A=TextureCoordinate(), B=NamedRerouteUsage())
  n35 = ScalarParameter() {defaultValue=0.05, parameterName=Offset}
  n36 = ComponentMask(None=VectorParameter() {parameterName=ColorNoiseBSpeed, defaultValue=<rgba 0 1 0 0>}.RGB) {g=true, r=true}
  n37 = TextureSampleParameter2D(UVs=Panner(Coordinate=Add(A=@n34, B=@n35), Speed=@n36) {speedY=1}) {parameterName=ColorNoiseB, texture=/Game/Effects/Fort_Effects/Textures/Tile/T_Noise_03}
  n38 = TextureCoordinate()
  n39 = NamedRerouteUsage()
  n40 = TextureSampleParameter2D(UVs=Panner(Coordinate=Add(A=@n38, B=@n39), Speed=@n36) {speedY=1}) {parameterName=ColorNoiseB, texture=&T_Noise_03}
  n41 = Add(A=TextureCoordinate(), B=NamedRerouteUsage())
  n42 = TextureSampleParameter2D(UVs=Panner(Coordinate=Subtract(A=@n41, B=@n35), Speed=@n36) {speedY=1}) {parameterName=ColorNoiseB, texture=&T_Noise_03}
  n43 = Power(
   Base=Desaturation(
    None=MaterialFunctionCall(X=@n37.R, Y=@n40.R, Z=@n42.R) {materialFunction=&MakeFloat3}.Result
    Fraction=ScalarParameter() {parameterName=Des}
   )
   Exp=ScalarParameter() {defaultValue=1, parameterName=ColorPower}
  ) {constExponent=0.8}
  n44 = VectorParameter() {parameterName=Color1, defaultValue=<rgba 2.949 3 2.9904 1>}
  n45 = ScalarParameter() {parameterName=ColorNoiseA_Min}
  n46 = ScalarParameter() {defaultValue=0.05, parameterName=Offset}
  n47 = Add(A=Add(A=TextureCoordinate(), B=NamedRerouteUsage()), B=@n46)
  n48 = ComponentMask(None=VectorParameter() {parameterName=ColorNoiseASpeed, defaultValue=<rgba 0.5 1 0 0>}.RGB) {g=true, r=true}
  n49 = LinearInterpolate(
   A=@n45
   Alpha=TextureSampleParameter2D(UVs=Panner(Coordinate=@n47, Speed=@n48) {speedX=0.5, speedY=1}) {parameterName=ColorNoiseA, samplerType=SAMPLERTYPE_Grayscale, texture=&LowResBlurredNoise}.R
  )
  n50 = Add(A=TextureCoordinate(), B=NamedRerouteUsage())
  n51 = LinearInterpolate(
   A=@n45
   Alpha=TextureSampleParameter2D(UVs=Panner(Coordinate=@n50, Speed=@n48) {speedX=0.5, speedY=1}) {parameterName=ColorNoiseA, samplerType=SAMPLERTYPE_Grayscale, texture=&LowResBlurredNoise}.R
  )
  n52 = Subtract(A=Add(A=TextureCoordinate(), B=NamedRerouteUsage()), B=@n46)
  n53 = LinearInterpolate(
   A=@n45
   Alpha=TextureSampleParameter2D(UVs=Panner(Coordinate=@n52, Speed=@n48) {speedX=0.5, speedY=1}) {parameterName=ColorNoiseA, samplerType=SAMPLERTYPE_Grayscale, texture=&LowResBlurredNoise}.R
  )
  n54 = Power(
   Base=Desaturation(
    None=MaterialFunctionCall(X=@n49, Y=@n51, Z=@n53) {materialFunction=&MakeFloat3}.Result
    Fraction=ScalarParameter() {parameterName=Des}
   )
   Exp=ScalarParameter() {defaultValue=1, parameterName=ColorPower}
  ) {constExponent=1}
  n55 = Multiply(A=Add(A=Multiply(A=@n33.RGB, B=@n43), B=Multiply(A=@n44.RGB, B=@n54)), B=Add(A=Multiply(A=@n32) {constB=3}) {constB=0})
  n56 = VectorParameter() {parameterName=MainColor, defaultValue=<rgba 1 1 1 0>}
  Reroute(None=MaterialFunctionCall(Emissive=Multiply(A=@n55, B=@n56.RGB)) {materialFunction=&MF_FX_EmissiveAdjustments}.Result)
 ]
}
