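# Niagara skydiving contrail cosmetic (SprocketPoppy Boat) with ghostly green skull meshes, ribbons, smoke, sparkles, airflow and glow with a light.
# Emitters follow skeletal mesh bones, using vortex, attraction and lift forces to swirl particles.
# BodySkull, SkullMeshes and Souce particles seed ribbon trails through particle attribute reads.
# A smoothed system value scales spawning and intensity, ParaGlideLeanAlpha gates smoke, and front-end modifiers adjust the effect outside gameplay.
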NiagaraSystem /BRCosmetics/Athena/Effects/Particles/Skydiving/NS_Contrail_SprocketPoppy_Boat/NS_Contrail_SprocketPoppyBoat
dependencies {
 usedDataInterfaces [
  &NiagaraDataInterfaceCurve
  &NiagaraDataInterfaceParticleRead
  &NiagaraDataInterfaceColorCurve
  &NiagaraDataInterfaceVectorCurve
  &NiagaraDataInterfaceArrayFloat3
  &NiagaraDataInterfacePlatformSet
  &NiagaraDataInterfaceSkeletalMesh
 ]
 usedDynamicInputs []
 usedModules [
  &self:SystemSpawnScript.SystemScriptSource.SystemScriptGraph.NiagaraNodeAssignment_0.SetVariables_#1
  &SystemState
  &self:SystemSpawnScript.SystemScriptSource.SystemScriptGraph.NiagaraNodeAssignment_1.SetVariables_#2
  &self:SystemSpawnScript.SystemScriptSource.SystemScriptGraph.NiagaraNodeAssignment_2.SetVariables_#3
 ]
}
emitters {
 BodySkull {
  emitterUpdateScript [
   &EmitterState()
   &SpawnBurst_Instantaneous(
    Spawn Count=2
    Spawn Probability=&FN_FrontEndFloatModifier(
     Gameplay=&MakeCustomFloatFromBool(Bool=&SetBoolByFloatComparison(A=$User.ParaGlideLeanAlpha, B=0), True Float=1, False Float=1)
     Front End=0
    )
    Use Spawn Probability=true
    UseLoopCountLimit=false
   )
   &SpawnRate(
    Spawn Probability=&FN_FrontEndFloatModifier(
     Gameplay=&MakeCustomFloatFromBool(Bool=&SetBoolByFloatComparison(A=$User.ParaGlideLeanAlpha, B=0), True Float=1, False Float=1)
     Front End=0
    )
    SpawnRate=3
    Use Spawn Probability=true
   )
  ]
  particleSpawnScript [
   &PartitionParticles(Number Of Partitions=4)
   &InitializeParticle(
    Color Mode=::NewEnumerator0
    Lifetime=0.7
    Mesh Scale Mode=::NewEnumerator1
    Mesh Uniform Scale Max=0.4
    Mesh Uniform Scale Min=0.3
    Sprite Size Mode=::NewEnumerator3
    Uniform Sprite Size=50
    UsePositionOffset=false
    Write Material Random=true
   )
   &InitialMeshOrientation(
    Mesh Orientation Mode=::NewEnumerator2
    Rotation=&SelectVectorFromArray(
     Direct Array Index=$Particles.PartitionParticles.Partition
     Vector Selection Array=DataInterface() {propertyValues {FloatData [<0 0 0>, <0 0 0>, <0 0.5 0>, <0 0.5 0>]}}
    )
    Use Rotation Vector=true
   )
   &ShapeLocation(
    MaskSpawnGroup=false
    Override Seed=false
    Shape Origin=&VectorFromFloat(Value=0)
    Sphere Radius=1
    UseOwnerScale=true
   )
   &InitialMeshOrientation(
    Mesh Orientation Mode=::NewEnumerator2
    Rotation=&MakeCustomVectorFromBool(
     Bool=&RandomBool()
     True Vector=&RandomRangeVector(Minimum=<0.05 0 0>, Maximum=<0.45 0 0>)
     False Vector=&RandomRangeVector(Minimum=<0.55 0 0>, Maximum=<0.95 0 0>)
    )
    Use Rotation Vector=true
   ) {enabled=false, module=InitialMeshOrientation001}
  ]
  particleUpdateScript [
   &ParticleState()
   &SkeletalMeshLocation(
    Bone / Socket Index=$Particles.PartitionParticles.Partition
    Bone Sampling Mode=::NewEnumerator4
    MaskSpawnGroup=false
    Position Sampling=::NewEnumerator1
    Sampled Position Offset=$Particles.Initial.ShapeLocation.ShapePosition
    Skeletal Mesh=$System.Skeletal Mesh
   )
   &SkeletalMeshLocation(
    MaskSpawnGroup=false
    Mesh Sampling Type=::NewEnumerator5
    Orientation Sampling=::NewEnumerator1
    Position Sampling=::NewEnumerator2
    Skeletal Mesh=$System.Skeletal Mesh
    Socket Sampling Mode=::NewEnumerator5
   ) {module=SkeletalMeshLocation001}
   &VortexForce(
    Origin Pull Only On Vortex Axis=false
    UseOriginOffset=false
    Vortex Force Amount=1000
    Vortex Origin:NiagaraPosition=$Output.SkeletalMeshLocation001.SampledPositionWithoutOffset
   )
   &ScaleVelocity()
   &GravityForce(Gravity=<0 0 5000>)
   /Niagara/Modules/Update/Forces/LineAttractionForce(
    Attraction Strength=300
    Line End:Vector3f=&AddVectorToPosition(Position=<0 0 0>, Vector=Linked() {linkedVariable {name=Engine.Emitter.SimulationPosition, type=NiagaraPosition}})
    Line Start:Vector3f=&AddVectorToPosition(Position=<0 0 0>, Vector=Linked() {linkedVariable {name=Engine.Emitter.SimulationPosition, type=NiagaraPosition}})
    Use Attraction Falloff=false
   )
   /SetVariables(Transient.PhysicsForce:Vector3f=&Multiply_Vector(A=$Transient.PhysicsForce, B=<25 25 0>)) {isSetParameters=true}
   &Drag(Drag=50, Use Linear Drag=true, Use Rotational Drag=true)
   &SolveForcesAndVelocity(Clamp Velocity=false, Limit Acceleration=false)
   &ScaleColor(
    Scale Alpha=&FloatFromCurve(FloatCurve=DataInterface() {propertyValues {Curve {keys [
         {interpMode=RCIM_Cubic, tangentMode=RCTM_User, time=0, value=0}
         {interpMode=RCIM_Cubic, tangentMode=RCTM_User, time=0.2, value=1}
         {tangentMode=RCTM_User, time=1, value=0}
        ]}}})
    ScaleA=true
    ScaleRGB=true
   )
  ]
  rendererClasses [&NiagaraMeshRendererProperties]
  renderers [{rendererClass=&NiagaraMeshRendererProperties, rendererIndex=0}]
 }
 BodySkullTrail {
  emitterUpdateScript [
   &EmitterState()
   &SpawnParticlesFromOtherEmitter(
    Attribute Reader=DataInterface() {propertyValues {EmitterBinding {bindingMode=Other, emitterName=BodySkull}}}
    Spawn Rate=30
    Use Maximum Age=false
    Use Minimum Age=false
   )
  ]
  particleSpawnScript [
   &InitializeParticle(Lifetime=0.6, Ribbon Width=20, Ribbon Width Mode=::NewEnumerator1, UsePositionOffset=false, Write Material Random=true)
   &SampleParticlesFromOtherEmitter(Dynamic Material Parameter 3 Sampling:ENiagara_DirectReadSamplingMode=::NewEnumerator1, Particle ID Sampling=::NewEnumerator1)
   /SetVariables(Particles.Position=&MakeVector(
     X=&MakeFloatFromVector4(Vector 4=$Output.SampleParticlesFromOtherEmitter.DynamicMaterialParameter3, Channel=::NewEnumerator0)
     Y=&MakeFloatFromVector4(Vector 4=$Output.SampleParticlesFromOtherEmitter.DynamicMaterialParameter3, Channel=::NewEnumerator1)
     Z=&MakeFloatFromVector4(Vector 4=$Output.SampleParticlesFromOtherEmitter.DynamicMaterialParameter3, Channel=::NewEnumerator2)
    )) {enabled=false, isSetParameters=true}
  ]
  particleUpdateScript [
   &SolveForcesAndVelocity(Clamp Velocity=false, Limit Acceleration=false)
   &ParticleState()
   /SetVariables(Particles.Position=$Particles.Position) {isSetParameters=true, module=SetVariables_#4}
   &RegionMask(Offset:Vector3f=<0 0 50>, Region Radius=150, UseOffset=true, UseRotation=false)
   &ScaleColor(
    Scale Alpha=&FloatFromCurve(
     FloatCurve=DataInterface() {propertyValues {Curve {keys [
         <at 0 0 RCIM_Cubic>
         {arriveTangent=-0.1027, interpMode=RCIM_Cubic, leaveTangent=-0.1027, tangentMode=RCTM_User, time=0.1, value=1}
         {arriveTangent=-1.6656, interpMode=RCIM_Cubic, leaveTangent=-1.6656, time=0.3762, value=0.3007}
         {arriveTangent=-0.4822, interpMode=RCIM_Cubic, leaveTangent=-0.4822, time=0.676, value=0.0406}
         <at 1 0 RCIM_Cubic>
        ]}}}
     Scale Curve=&Multiply_Float(A=&FN_FrontEndFloatModifier(Gameplay=6, Front End=2), B=$Particles.RegionMask.Float)
    )
    Scale RGB=&MakeVectorFromLinearColorRGB(Color=&ColorFromCurve(ColorCurve=DataInterface() {propertyValues {
        AlphaCurve {keys [<at 0 1 RCIM_Cubic>,<at 1 0 RCIM_Cubic>]}
        BlueCurve {keys [<at 0 0.4624 RCIM_Cubic>,<at 0.3794 0.2549>,<at 0.6536 0.1275 RCIM_Cubic>]}
        GreenCurve {keys [<at 0 1 RCIM_Cubic>,<at 0.3794 1>,<at 0.6536 1 RCIM_Cubic>]}
        RedCurve {keys [<at 0 0 RCIM_Cubic>,<at 0.3794 0>,<at 0.6536 0 RCIM_Cubic>]}
       }}))
    ScaleA=true
    ScaleRGB=true
   )
   &ScaleRibbonWidth(Ribbon Width Scale=&FloatFromCurve(FloatCurve=DataInterface() {propertyValues {Curve {keys [<at 0 1>,<at 1 0>]}}}, Scale Curve=0.8))
  ]
  rendererClasses [&NiagaraRibbonRendererProperties]
  renderers [{rendererClass=&NiagaraRibbonRendererProperties}]
 }
 BottomAirflow {
  emitterUpdateScript [
   &EmitterState(Life Cycle Mode=::NewEnumerator1, Loop Behavior=::NewEnumerator1, Loop Duration Mode=::NewEnumerator1, UseLoopDelay=false)
   &SpawnRate(
    Spawn Probability=&Lerp_Float(A=0, B=0.8, Alpha=$System.InterpolateOverTime.InterpolateOverTime_Float.Moving Average)
    SpawnRate=10
    Use Spawn Probability=true
   )
   /SetVariables(Emitter.Width=20) {isSetParameters=true}
  ]
  particleSpawnScript [
   &InitializeParticle(
    Color=<rgba 0.9062 1 0.9154 1>
    Lifetime Max=0.8
    Lifetime Min=0.4
    Lifetime Mode=::NewEnumerator1
    Mesh Scale=&Multiply_VectorByFloat(
     Vector=&RandomRangeVector(Minimum=<1 1 0.5>, Maximum=<1 1 0.5>)
     Float=&RandomRangeFloat(Minimum=0.7, Maximum=1)
    )
    Mesh Scale Mode=::NewEnumerator4
    Position Offset=<0 0 115>
    Sprite Size Mode=::NewEnumerator3
    Uniform Sprite Size=5
    UsePositionOffset=true
    Write Material Random=true
   )
   &CameraOffset(Camera Offset Amount=30)
   &InitialMeshOrientation(
    Mesh Orientation Mode=::NewEnumerator2
    Rotation=&RandomRangeVector(Minimum=<0 0 -1>, Maximum=<0 0 1>)
    Use Rotation Vector=true
   )
   &InitialMeshOrientation(Mesh Orientation Mode=::NewEnumerator2, Use Rotation Vector=true) {module=InitialMeshOrientation001}
  ]
  particleUpdateScript [
   &SolveForcesAndVelocity(Clamp Velocity=false, Limit Acceleration=false)
   &SkeletalMeshLocation(
    Bone / Socket Index=5
    Bone Sampling Mode=::NewEnumerator4
    MaskSpawnGroup=false
    Orientation Sampling=::NewEnumerator1
    Sampled Position Offset=&AddVectorToPosition(Position=$System.FX offset, Vector=<0 0 -20>)
    Skeletal Mesh=$System.Skeletal Mesh
   )
   &ParticleState()
   &ScaleMeshSize(Scale Factor=&Multiply_Vector(
     A=&RandomRangeVector(Minimum=<1 1 1>, Maximum=<1 1 2>)
     B=&VectorFromCurve(
      VectorCurve=DataInterface() {propertyValues {XCurve {keys [<at 0 1.3>,<at 1 0.9>]}, YCurve {keys [<at 0 1.3>,<at 1 0.9>]}, ZCurve {keys [<at 0 1.3>,<at 1 1.3>]}}}
      Scale Curve=$System.SizeScale
     )
    ))
   &UpdateMeshOrientation(Rotation Rate=&RandomRangeFloat(Minimum=0.5, Maximum=0.5), Rotation Vector=<0 0 1>)
   &ScaleColor(
    Scale Alpha=&FloatFromCurve(
     FloatCurve=DataInterface() {propertyValues {Curve {keys [
         {interpMode=RCIM_Cubic, tangentMode=RCTM_User, time=0, value=0}
         {interpMode=RCIM_Cubic, tangentMode=RCTM_User, time=0.5, value=1}
         {tangentMode=RCTM_User, time=1, value=0}
        ]}}}
     Scale Curve=&Multiply_Float(A=&FN_FrontEndFloatModifier(Gameplay=0.6, Front End=0.3), B=1)
    )
    ScaleA=true
    ScaleRGB=true
   )
   &DynamicMaterialParameters(
    Index 0 Param 1=&Multiply_Float(
     A=&RandomRangeFloat(Minimum=0.6, Maximum=1)
     B=&FN_FrontEndFloatModifier(Gameplay=1, Front End=0.5)
    )
    Module.Index0Float0Write=true
    Module.Index0Float1Write=true
    Module.Index0Float2Write=true
    Module.Index0Float3Write=true
   )
  ]
  rendererClasses [&NiagaraMeshRendererProperties]
  renderers [{rendererClass=&NiagaraMeshRendererProperties, rendererIndex=0}]
 }
 BottomSmoke {
  emitterUpdateScript [
   &EmitterState()
   &SpawnRate(
    Spawn Probability=&MakeCustomFloatFromBool(Bool=&SetBoolByFloatComparison(A=$User.ParaGlideLeanAlpha, B=0), True Float=1, False Float=0)
    SpawnRate=20
    Use Spawn Probability=true
   )
   /SetVariables(Emitter.ColorDes=-2) {isSetParameters=true, module=SetVariables_#5}
  ]
  particleSpawnScript [
   &InitializeParticle(
    Color=&MakeCustomLinearColorFromBool(Bool=&RandomBool(), True Color=<rgba 0 1.1328 2 1>, False Color=<rgba 1 1 1 1>)
    Lifetime Max=1.3
    Lifetime Min=1.3
    Lifetime Mode=::NewEnumerator1
    Position Offset=&MakeVector(X=0, Y=0, Z=0)
    Sprite Rotation Mode=::NewEnumerator2
    Sprite Size Mode=::NewEnumerator1
    Uniform Sprite Size Max=100
    Uniform Sprite Size Min=50
    UsePositionOffset=false
    Write Material Random=true
   )
   &ShapeLocation(
    Handle Radius=0
    Large Radius=110
    MaskSpawnGroup=false
    Offset=&MakeVector(X=0, Y=0, Z=&FN_FrontEndFloatModifier(Gameplay=150, Front End=90))
    Offset Mode=::NewEnumerator0
    Override Seed=false
    Shape Origin=&VectorFromFloat(Value=0)
    Shape Primitive=::NewEnumerator3
    Surface Distribution=1
    UseOwnerScale=true
   )
   /SetVariables(Particles.SpriteSize=&Multiply_Vector2D_ByFloat(
     Float=&FloatFromCurve(
      FloatCurve=DataInterface() {propertyValues {Curve {keys [{interpMode=RCIM_Cubic, tangentMode=RCTM_Break, time=0, value=1}, <at 1 0.3>]}}}
      CurveIndex=&Divide_Float(
       A=&Subtract_Float(A=&MakeFloatFromVector(Vector=$Particles.ShapeLocation.ShapePosition, Channel=::NewEnumerator2), B=100)
       B=170
      )
     )
     Vector2D=$Particles.SpriteSize
    )) {enabled=false, isSetParameters=true, module=SetVariables_#6}
  ]
  particleUpdateScript [
   &SkeletalMeshLocation(
    Bone / Socket Index=5
    Bone Sampling Mode=::NewEnumerator4
    MaskSpawnGroup=false
    Orientation Sampling=::NewEnumerator1
    Position Sampling=::NewEnumerator1
    Sampled Position Offset=&AddVectorToPosition(Position=$Particles.Initial.ShapeLocation.ShapePosition, Vector=<0 0 -50>)
    Skeletal Mesh=$System.Skeletal Mesh
   )
   &ParticleState()
   &SpriteRotationRate(Rotation Rate=&RandomRangeFloat(Minimum=-150, Maximum=150))
   &ScaleSpriteSize(Uniform Curve Sprite Scale=DataInterface() {propertyValues {Curve {keys [
        {arriveTangent=-0.321, interpMode=RCIM_Cubic, leaveTangent=-0.321, tangentMode=RCTM_User, time=0, value=1.3}
        {arriveTangent=-0.6, interpMode=RCIM_Cubic, leaveTangent=-0.6, time=0.1961, value=1.1888}
        {arriveTangent=-0.2921, tangentMode=RCTM_User, time=1, value=0.7}
       ]}}})
   &GravityForce(Gravity=<0 0 200>)
   &PointForce(
    Force Origin=&AddVectorToPosition(Position=$Output.SkeletalMeshLocation.SampledPositionWithoutOffset, Vector=<0 0 150>)
    Force Strength=-300
    Module.Use Falloff=false
    Module.UseLinearForce=false
   )
   &SolveForcesAndVelocity(Clamp Velocity=false, Limit Acceleration=false)
   &ScaleColor(
    Scale Alpha=&FloatFromCurve(
     FloatCurve=DataInterface() {propertyValues {Curve {keys [
         {interpMode=RCIM_Cubic, tangentMode=RCTM_User, time=0, value=0}
         {arriveTangent=2.5, interpMode=RCIM_Cubic, leaveTangent=2.5, time=0.2402, value=0.2722}
         {interpMode=RCIM_Cubic, tangentMode=RCTM_User, time=0.4, value=1}
         {arriveTangent=-1.6667, interpMode=RCIM_Cubic, leaveTangent=-1.6667, time=0.6013, value=0.1294}
         {tangentMode=RCTM_User, time=1, value=0}
        ]}}}
     Scale Curve=&Multiply_Float(
      A=&FN_FrontEndFloatModifier(Gameplay=1.4, Front End=0.5)
      B=&ClampFloat(Float=&Lerp_Float(A=0, B=1, Alpha=$System.InterpolateOverTime.InterpolateOverTime_Float.Moving Average))
     )
    )
    Scale RGB=&MakeVectorFromLinearColorRGB(Color=&ColorFromCurve(ColorCurve=DataInterface() {propertyValues {
        AlphaCurve {keys [<at 0 1 RCIM_Cubic>,<at 1 0 RCIM_Cubic>]}
        BlueCurve {keys [<at 0.1979 0.0168 RCIM_Cubic>,<at 0.244 0.0669>,<at 1 0.2708 RCIM_Cubic>]}
        GreenCurve {keys [<at 0.1979 0.03 RCIM_Cubic>,<at 0.244 0.1578>,<at 1 2 RCIM_Cubic>]}
        RedCurve {keys [<at 0.1979 0.0036 RCIM_Cubic>,<at 0.244 0.0214>,<at 1 0.5938 RCIM_Cubic>]}
       }}))
    ScaleA=true
    ScaleRGB=true
   )
   &DynamicMaterialParameters(
    Index 0 Param 1=0.75
    Module.Index0Float0Write=true
    Module.Index0Float1Write=true
    Module.Index0Float2Write=true
    Module.Index0Float3Write=true
   )
  ]
  rendererClasses [&NiagaraSpriteRendererProperties]
  renderers [{rendererClass=&NiagaraSpriteRendererProperties}]
 }
 GhostTrail {
  emitterUpdateScript [
   &EmitterState()
   &SpawnParticlesFromOtherEmitter(
    Attribute Reader=DataInterface() {propertyValues {EmitterBinding {bindingMode=Other, emitterName=Souce}}}
    Spawn Rate=30
    Spawn Rate Per Particle Cap=60
    Use Maximum Age=false
    Use Minimum Age=false
   )
  ]
  particleSpawnScript [
   &InitializeParticle(
    Color=<rgba 0.4847 1.2 0.3615 1>
    Lifetime=0.4
    Ribbon Width=15
    Ribbon Width Mode=::NewEnumerator1
    UsePositionOffset=false
    Write Material Random=true
   )
   &SampleParticlesFromOtherEmitter(Dynamic Material Parameter 3 Sampling:ENiagara_DirectReadSamplingMode=::NewEnumerator1, Particle ID Sampling=::NewEnumerator1)
   /SetVariables(Particles.RibbonWidth=&Multiply_Float(
     A=$Particles.RibbonWidth
     B=&MakeFloatFromVector4(Vector 4=$Output.SampleParticlesFromOtherEmitter.DynamicMaterialParameter3, Channel=::NewEnumerator3)
    )) {isSetParameters=true}
  ]
  particleUpdateScript [
   &SolveForcesAndVelocity(Clamp Velocity=false, Limit Acceleration=false)
   &ParticleState()
   /SetVariables(Particles.Position=$Particles.Position) {isSetParameters=true, module=SetVariables_#4}
   &DynamicMaterialParameters(
    Index 0 Param 1=-0.5
    Index 0 Param 2=0.5
    Index 0 Param 3=1
    Module.Index0Float0Write=true
    Module.Index0Float1Write=true
    Module.Index0Float2Write=true
    Module.Index0Float3Write=true
   )
   &ScaleRibbonWidth(Ribbon Width Scale=&FloatFromCurve(
     FloatCurve=DataInterface() {propertyValues {Curve {keys [
         {arriveTangent=10.6409, interpMode=RCIM_Cubic, time=0, value=1.5}
         {arriveTangent=-0.5, interpMode=RCIM_Cubic, leaveTangent=-0.5, tangentMode=RCTM_Break, time=0.2, value=1.525}
         {arriveTangent=-2.0489, interpMode=RCIM_Cubic, leaveTangent=-2.0489, tangentMode=RCTM_User, time=1, value=1}
        ]}}}
     Scale Curve=&FN_FrontEndFloatModifier(Gameplay=1, Front End=0.9)
    ))
   &RegionMask(Offset:Vector3f=<0 0 175>, Region Radius=180, UseOffset=true, UseRotation=false)
   &ScaleColor(
    Scale Alpha=&Multiply_Float(
     A=&FloatFromCurve(FloatCurve=DataInterface() {propertyValues {Curve {keys [
          {interpMode=RCIM_Cubic, tangentMode=RCTM_User, time=0, value=0}
          {arriveTangent=2, interpMode=RCIM_Cubic, leaveTangent=2, time=0.1069, value=1.002}
          {interpMode=RCIM_Cubic, tangentMode=RCTM_User, time=0.5, value=1}
          {tangentMode=RCTM_User, time=1, value=1}
         ]}}})
     B=$Particles.RegionMask.Float
    )
    Scale RGB=&MakeVectorFromLinearColorRGB(Color=&Multiply_LinearColorByFloat(
      Linear Color=&ColorFromCurve(ColorCurve=DataInterface() {propertyValues {
         AlphaCurve {keys [<at 0 1 RCIM_Cubic>,<at 1 0 RCIM_Cubic>]}
         BlueCurve {keys [<at 0 0.5361 RCIM_Cubic>,<at 1 0.6774 RCIM_Cubic>]}
         GreenCurve {keys [<at 0 1 RCIM_Cubic>,<at 1 1 RCIM_Cubic>]}
         RedCurve {keys [<at 0 0.768 RCIM_Cubic>,<at 1 0.2145 RCIM_Cubic>]}
        }})
      Float=1
     ))
    ScaleA=true
    ScaleRGB=true
   )
  ]
  rendererClasses [&NiagaraRibbonRendererProperties]
  renderers [{rendererClass=&NiagaraRibbonRendererProperties}]
 }
 Glow {
  emitterUpdateScript [
   &EmitterState(
    Inactive Response=::NewEnumerator1
    Life Cycle Mode=::NewEnumerator1
    Loop Behavior=::NewEnumerator1
    Loop Duration Mode=::NewEnumerator1
    UseLoopDelay=false
   )
   &SpawnBurst_Instantaneous(Use Spawn Probability=false, UseLoopCountLimit=false)
   /SetVariables(Emitter.AngleFade:NiagaraFloat=&FN_PickFloatDependOnPlatform(TargetPlatformActive=1, TargetPlatformNoActive=2)) {isSetParameters=true}
  ]
  particleSpawnScript [
   &InitializeParticle(
    Color=<rgba 0 0.07 0.035 1.5>
    Position Offset=<0 0 130>
    Sprite Size=<200 500>
    Sprite Size Mode=::NewEnumerator4
    UsePositionOffset=false
    Write Material Random=true
   )
   &CameraOffset()
   &SpriteFacingAndAlignment(Sprite Alignment=<0 0 1>, WriteAlignment=true, WriteFacing=true)
  ]
  particleUpdateScript [
   /SetVariables(Particles.SizeScale:Vector2f=&Multiply_Vector2D_ByFloat(
     Float=1.1
     Vector2D=&MakeVector2D(X=1, Y=&Lerp_Float(A=0.7, B=1.2, Alpha=$System.InterpolateOverTime.InterpolateOverTime_Float.Moving Average))
    )) {isSetParameters=true}
   &SkeletalMeshLocation(
    Bone / Socket Index=5
    Bone Sampling Mode=::NewEnumerator4
    MaskSpawnGroup=false
    Orientation Sampling=::NewEnumerator1
    Sampled Position Offset=&AddVectorToPosition(
     Position=&FN_FrontEndVector3Modifier(Gameplay=<0 0 180>, Front End=<0 0 160>)
     Vector=&MakeVector(
      X=0
      Y=0
      Z=&ClampFloat(
       Float=&Lerp_Float(A=-20, B=40, Alpha=$System.InterpolateOverTime.InterpolateOverTime_Float.Moving Average)
       Min=-20
       Max=40
      )
     )
    )
    Skeletal Mesh=$System.Skeletal Mesh
   )
   &ParticleState(Kill Particles When Lifetime Has Elapsed=false)
   &ScaleSpriteSize(Scale Factor=$Particles.SizeScale, Scale Sprite Size Mode=::NewEnumerator2)
   &DynamicMaterialParameters(
    Index 0 Param 1=2
    Index 0 Param 2=10
    Module.Index0Float0Write=true
    Module.Index0Float1Write=true
    Module.Index0Float2Write=true
    Module.Index0Float3Write=true
   )
   &ScaleColor(
    Scale Alpha=&Multiply_Float(
     A=&Lerp_Float(A=0.5, B=1, Alpha=$System.InterpolateOverTime.InterpolateOverTime_Float.Moving Average)
     B=&FN_FrontEndFloatModifier(Gameplay=1, Front End=0.3)
    )
    ScaleA=true
    ScaleRGB=true
   )
   &FN_Sprite_Attributes(
    Color=&MakeLinearColorFromVectorAndFloat(
     "Vector (RGB)"=&MakeVectorFromLinearColorRGB(Color=<rgba 0.0808 1 0.3106 1>)
     "Float (Alpha)"=&Multiply_Float(
      A=&FN_FrontEndFloatModifier(
       Gameplay=&Lerp_Float(A=0.3, B=0.35, Alpha=$System.InterpolateOverTime.InterpolateOverTime_Float.Moving Average)
       Front End=0.1
      )
      B=1
     )
    )
    Create Color Attribute=true
    Create Position Attribute=true
    Create SpriteSize Attribute=true
    SpriteSize=&Multiply_Vector2D(A=<120 400>, B=$Particles.SizeScale)
    Write to Module Namespace=true
   )
   &FN_Light_Attributes(
    Color=<rgba 0.2083 10 0.2083 1>
    Create Color Attribute=true
    Create Exponent Attribute=true
    Create Position Attribute=true
    Create Radius Attribute=true
    Exponent:NiagaraFloat=&FN_FrontEndFloatModifier(Gameplay=3, Front End=3)
    Position=&AddVectorToPosition(
     Position=$Particles.Position
     Vector=&FN_FrontEndVector3Modifier(
      Gameplay=&Lerp_Vector(A=<0 0 0>, B=<0 0 -90>, Alpha=$System.InterpolateOverTime.InterpolateOverTime_Float.Moving Average)
      Front End=<0 0 0>
     )
    )
    Radius=130
    Scale Color=&Lerp_Float(A=1, B=2, Alpha=$System.InterpolateOverTime.InterpolateOverTime_Float.Moving Average)
   )
  ]
  rendererClasses [&NiagaraSpriteRendererProperties, &NiagaraLightRendererProperties]
  renderers [
   {rendererClass=&NiagaraSpriteRendererProperties}
   {rendererClass=&NiagaraSpriteRendererProperties, rendererIndex=1}
   {rendererClass=&NiagaraLightRendererProperties, rendererIndex=2}
  ]
 }
 SkullMeshes {
  emitterUpdateScript [
   &EmitterState()
   &SpawnRate(
    Spawn Probability=&ClampFloat(Float=&Lerp_Float(A=0.6, B=1, Alpha=$System.InterpolateOverTime.InterpolateOverTime_Float.Moving Average))
    SpawnRate=10
    Use Spawn Probability=true
   )
  ]
  particleSpawnScript [
   &InitializeParticle(
    Color=<rgba 1.3 1.3 1.3 2>
    Lifetime=1.5
    Mesh Scale=&Multiply_VectorByFloat(Vector=<1 1 1>, Float=&RandomRangeFloat(Minimum=0.25, Maximum=0.3))
    Mesh Scale Mode=::NewEnumerator4
    Sprite Size Mode=::NewEnumerator3
    Uniform Sprite Size=5
    UsePositionOffset=false
    Write Material Random=true
   )
   &ShapeLocation(
    Handle Radius=60
    Large Radius=90
    MaskSpawnGroup=false
    Offset=$System.FX offset
    Offset Mode=::NewEnumerator0
    Override Seed=false
    Shape Origin=&VectorFromFloat(Value=0)
    Shape Primitive=::NewEnumerator3
    UseOwnerScale=true
   )
   &CameraOffset()
   &InitialMeshOrientation(Mesh Orientation Mode=::NewEnumerator2, Rotation=<0 0.25 0>, Use Rotation Vector=true)
   &InitialMeshOrientation(
    Mesh Orientation Mode=::NewEnumerator2
    Rotation=&MakeCustomVectorFromBool(
     Bool=&RandomBool()
     True Vector=&RandomRangeVector(Minimum=<0.05 0 0>, Maximum=<0.45 0 0>)
     False Vector=&RandomRangeVector(Minimum=<0.55 0 0>, Maximum=<0.95 0 0>)
    )
    Use Rotation Vector=true
   ) {module=InitialMeshOrientation001}
  ]
  particleUpdateScript [
   &SkeletalMeshLocation(
    Bone / Socket Index=5
    Bone Sampling Mode=::NewEnumerator4
    MaskSpawnGroup=false
    Position Sampling=::NewEnumerator1
    Sampled Position Offset:NiagaraPosition=$Particles.Initial.ShapeLocation.ShapePosition
    Skeletal Mesh=$System.Skeletal Mesh
   )
   &VortexForce(
    UseOriginOffset=false
    Vortex Force Amount=&Multiply_Float(A=500, B=1.5)
    Vortex Origin:NiagaraPosition=$Output.SkeletalMeshLocation.SampledPositionWithoutOffset
    Vortex Origin Offset:Vector3f=&RandomRangeVector(Minimum=<0 0 0>, Maximum=<0 0 0>)
   )
   &ScaleVelocity(Velocity Scale=&Multiply_VectorByFloat(Vector=<1 1 1>, Float=1))
   &LineAttractionForce(
    Attraction Strength=1000
    Line End:Vector3f=&AddVectorToPosition(Position=<0 0 500>, Vector=$Output.SkeletalMeshLocation.SampledPositionWithoutOffset)
    Line Start:Vector3f=&AddVectorToPosition(Position=<0 0 500>, Vector=$Output.SkeletalMeshLocation.SampledPositionWithoutOffset)
    Use Attraction Falloff=false
   )
   &LinearForce(Force=<0 0 8000>)
   /Niagara/Modules/Update/Forces/SpringForce(
    " Particle Equilibrium Position":NiagaraPosition=$Particles.SkeletalMeshLocation.SampledPosition
    Attachment Point Velocity:Vector3f=<0 0 500>
    Desired Minimum Separation Distance=1
    Force Strength=-0.1
    Spring Tightness=7
   )
   /SetVariables(Transient.PhysicsForce:Vector3f=&Multiply_Vector(
     A=$Transient.PhysicsForce
     B=&MakeVector(
      X=40
      Y=40
      Z=&ClampFloat(
       Float=&Lerp_Float(
        A=&RandomRangeFloat(Minimum=0.2, Maximum=0.4)
        B=1.5
        Alpha=$System.InterpolateOverTime.InterpolateOverTime_Float.Moving Average
       )
       Max=2
      )
     )
    )) {isSetParameters=true, module=SetVariables_#7}
   &Drag(Drag=40, Use Linear Drag=true, Use Rotational Drag=true)
   &SolveForcesAndVelocity(Clamp Velocity=false, Limit Acceleration=false)
   /SetVariables(Particles.Position=$Particles.Position) {isSetParameters=true, module=SetVariables_#4}
   &ParticleState()
   &ScaleColor(
    Scale Alpha=&FloatFromCurve(
     FloatCurve=DataInterface() {propertyValues {Curve {keys [
         {interpMode=RCIM_Cubic, tangentMode=RCTM_User, time=0, value=0}
         {interpMode=RCIM_Cubic, tangentMode=RCTM_User, time=0.5, value=1}
         {tangentMode=RCTM_User, time=1, value=0}
        ]}}}
     Scale Curve=2
    )
    ScaleA=true
    ScaleRGB=true
   )
   &DynamicMaterialParameters(
    Index 3 Param 3=&RandomRangeFloat(Minimum=0.7, Maximum=0.9)
    Index 3 Param 4=&RandomRangeFloat(Minimum=0, Maximum=1)
    Module.Index3Float0Write=true
    Module.Index3Float1Write=true
    Module.Index3Float2Write=true
    Module.Index3Float3Write=true
    Write Parameter Index 0=::NewEnumerator0
    Write Parameter Index 3=::NewEnumerator1
   )
  ]
  rendererClasses [&NiagaraMeshRendererProperties]
  renderers [{rendererClass=&NiagaraMeshRendererProperties, rendererIndex=0}]
 }
 SkullTrail {
  emitterUpdateScript [
   &EmitterState()
   &SpawnParticlesFromOtherEmitter(
    Attribute Reader=DataInterface() {propertyValues {EmitterBinding {bindingMode=Other, emitterName=SkullMeshes}}}
    Spawn Rate=30
    Use Maximum Age=false
    Use Minimum Age=false
   )
  ]
  particleSpawnScript [
   &InitializeParticle(Lifetime=0.6, Ribbon Width=17, Ribbon Width Mode=::NewEnumerator1, UsePositionOffset=false, Write Material Random=true)
   &SampleParticlesFromOtherEmitter(Dynamic Material Parameter 3 Sampling:ENiagara_DirectReadSamplingMode=::NewEnumerator1, Particle ID Sampling=::NewEnumerator1)
   &DynamicMaterialParameters(
    Index 0 Param 1=&MakeFloatFromVector4(Vector 4=$Output.SampleParticlesFromOtherEmitter.DynamicMaterialParameter3, Channel=::NewEnumerator3)
    Module.Index0Float0Write=true
    Module.Index0Float1Write=true
    Module.Index0Float2Write=true
    Module.Index0Float3Write=true
   )
   /SetVariables(Particles.RibbonWidth=&Multiply_Float(
     A=$Particles.RibbonWidth
     B=&MakeFloatFromVector4(Vector 4=$Output.SampleParticlesFromOtherEmitter.DynamicMaterialParameter3, Channel=::NewEnumerator2)
    )) {isSetParameters=true}
  ]
  particleUpdateScript [
   &SolveForcesAndVelocity(Clamp Velocity=false, Limit Acceleration=false)
   &ParticleState()
   /SetVariables(Particles.Position=$Particles.Position) {isSetParameters=true, module=SetVariables_#4}
   &RegionMask(Offset:Vector3f=<0 0 200>, Region Radius=150, UseOffset=true, UseRotation=false)
   &ScaleColor(
    Scale Alpha=&FloatFromCurve(
     FloatCurve=DataInterface() {propertyValues {Curve {keys [
         <at 0 0 RCIM_Cubic>
         {arriveTangent=-0.1027, interpMode=RCIM_Cubic, leaveTangent=-0.1027, tangentMode=RCTM_User, time=0.1015, value=0.5006}
         {arriveTangent=-1.1256, interpMode=RCIM_Cubic, leaveTangent=-1.1256, time=0.291, value=0.5653}
         {arriveTangent=-0.7973, interpMode=RCIM_Cubic, leaveTangent=-0.7973, time=0.4678, value=0.0884}
         <at 1 0 RCIM_Cubic>
        ]}}}
     Scale Curve=&Multiply_Float(A=6, B=$Particles.RegionMask.Float)
    )
    Scale RGB=&MakeVectorFromLinearColorRGB(Color=&ColorFromCurve(ColorCurve=DataInterface() {propertyValues {
        AlphaCurve {keys [<at 0 1 RCIM_Cubic>,<at 1 0 RCIM_Cubic>]}
        BlueCurve {keys [<at 0 0.4624 RCIM_Cubic>,<at 0.3794 0.2549>,<at 0.6536 0.1275 RCIM_Cubic>]}
        GreenCurve {keys [<at 0 1 RCIM_Cubic>,<at 0.3794 1>,<at 0.6536 1 RCIM_Cubic>]}
        RedCurve {keys [<at 0 0 RCIM_Cubic>,<at 0.3794 0>,<at 0.6536 0 RCIM_Cubic>]}
       }}))
    ScaleA=true
    ScaleRGB=true
   )
   &ScaleRibbonWidth(Ribbon Width Scale=&FloatFromCurve(FloatCurve=DataInterface() {propertyValues {Curve {keys [<at 0 1>,<at 1 0>]}}}))
  ]
  rendererClasses [&NiagaraRibbonRendererProperties]
  renderers [{rendererClass=&NiagaraRibbonRendererProperties}]
 }
 Souce {
  emitterUpdateScript [
   &EmitterState()
   &SpawnRate(
    Spawn Probability=&Lerp_Float(A=0, B=1, Alpha=$System.InterpolateOverTime.InterpolateOverTime_Float.Moving Average)
    SpawnRate=10
    Use Spawn Probability=true
   )
  ]
  particleSpawnScript [
   &InitializeParticle(
    Color=<rgba 0.099 0.099 0.099 1>
    Mesh Scale=&Multiply_VectorByFloat(Vector=<1 1 1>, Float=&RandomRangeFloat(Minimum=0.5, Maximum=0.5))
    Mesh Scale Mode=::NewEnumerator4
    Sprite Size Mode=::NewEnumerator3
    Uniform Sprite Size=5
    UsePositionOffset=false
    Write Material Random=true
   )
   &ShapeLocation(
    Handle Radius=0
    Large Radius=&FN_FrontEndFloatModifier(Gameplay=100, Front End=60)
    MaskSpawnGroup=false
    Offset=$System.FX offset
    Offset Mode=::NewEnumerator0
    Override Seed=false
    Shape Origin=&VectorFromFloat(Value=0)
    Shape Primitive=::NewEnumerator3
    UseOwnerScale=true
   )
   &CameraOffset()
   &InitialMeshOrientation(Mesh Orientation Mode=::NewEnumerator2, Rotation=<0 0 -0.25>, Use Rotation Vector=true)
  ]
  particleUpdateScript [
   &SkeletalMeshLocation(
    Bone / Socket Index=5
    Bone Sampling Mode=::NewEnumerator4
    MaskSpawnGroup=false
    Orientation Sampling=::NewEnumerator1
    Position Sampling=::NewEnumerator1
    Sampled Position Offset=$Particles.Initial.ShapeLocation.ShapePosition
    Skeletal Mesh=$System.Skeletal Mesh
   )
   &VortexForce(
    UseOriginOffset=false
    Vortex Force Amount=&Multiply_Float(A=500, B=1)
    Vortex Origin:NiagaraPosition=$Output.SkeletalMeshLocation.SampledPositionWithoutOffset
    Vortex Origin Offset:Vector3f=&RandomRangeVector(Minimum=<0 0 0>, Maximum=<0 0 0>)
   )
   &ScaleVelocity(Velocity Scale=&Multiply_VectorByFloat(Vector=<1 1 1>, Float=1))
   &LineAttractionForce(
    Attraction Strength=1000
    Line End:Vector3f=&AddVectorToPosition(Position=<0 0 500>, Vector=$Output.SkeletalMeshLocation.SampledPositionWithoutOffset)
    Line Start:Vector3f=&AddVectorToPosition(Position=<0 0 500>, Vector=$Output.SkeletalMeshLocation.SampledPositionWithoutOffset)
    Use Attraction Falloff=false
   )
   &LinearForce(Force=&RandomRangeVector(Minimum=<0 0 10000>, Maximum=<0 0 12000>))
   &SpringForce(
    " Particle Equilibrium Position":NiagaraPosition=$Particles.SkeletalMeshLocation.SampledPosition
    Attachment Point Velocity:Vector3f=<0 0 500>
    Desired Minimum Separation Distance=1
    Force Strength=-0.1
    Spring Tightness=10
   )
   /SetVariables(Transient.PhysicsForce:Vector3f=&Multiply_Vector(A=$Transient.PhysicsForce, B=<40 40 1>)) {isSetParameters=true, module=SetVariables_#7}
   &Drag(Drag=&RandomRangeFloat(Minimum=10, Maximum=10), Use Linear Drag=true, Use Rotational Drag=true)
   &SolveForcesAndVelocity(Clamp Velocity=false, Limit Acceleration=false)
   /SetVariables(Particles.Position=$Particles.Position) {isSetParameters=true, module=SetVariables_#4}
   &ParticleState()
   &DynamicMaterialParameters(
    Index 3 Param 4=&RandomRangeFloat(Minimum=0.8, Maximum=1.2)
    Module.Index3Float0Write=true
    Module.Index3Float1Write=true
    Module.Index3Float2Write=true
    Module.Index3Float3Write=true
    Write Parameter Index 0=::NewEnumerator0
    Write Parameter Index 3=::NewEnumerator1
   )
  ]
  rendererClasses [&NiagaraSpriteRendererProperties]
  renderers [{rendererClass=&NiagaraSpriteRendererProperties}]
 }
 Sparkles {
  emitterUpdateScript [
   &EmitterState()
   &SpawnRate(
    Spawn Probability=&ClampFloat(Float=&Lerp_Float(A=0.6, B=1, Alpha=$System.InterpolateOverTime.InterpolateOverTime_Float.Moving Average))
    SpawnRate=30
    Use Spawn Probability=true
   )
  ]
  particleSpawnScript [
   &InitializeParticle(
    Color=<rgba 2.775 5 0.55 1>
    Lifetime=2
    Sprite Size Max=<0.3 4>
    Sprite Size Min=<0.3 2>
    Sprite Size Mode=::NewEnumerator2
    UsePositionOffset=false
    Write Material Random=true
   )
   &ShapeLocation(
    Handle Radius=10
    Large Radius=50
    MaskSpawnGroup=false
    Offset=<0 0 100>
    Offset Mode=::NewEnumerator0
    Override Seed=false
    Shape Origin=&VectorFromFloat(Value=0)
    Shape Primitive=::NewEnumerator3
    Surface Distribution=1
    UseOwnerScale=true
   )
   &CameraOffset()
   &InitialMeshOrientation(Mesh Orientation Mode=::NewEnumerator2, Rotation=<0 0.25 0>, Use Rotation Vector=true)
  ]
  particleUpdateScript [
   &SkeletalMeshLocation(
    Bone / Socket Index=5
    Bone Sampling Mode=::NewEnumerator4
    MaskSpawnGroup=false
    Orientation Sampling=::NewEnumerator1
    Position Sampling=::NewEnumerator1
    Sampled Position Offset:NiagaraPosition=$Particles.Initial.ShapeLocation.ShapePosition
    Skeletal Mesh=$System.Skeletal Mesh
   ) {module=SkeletalMeshLocation001}
   &SkeletalMeshLocation(
    MaskSpawnGroup=false
    Mesh Sampling Type=::NewEnumerator5
    Orientation Sampling=::NewEnumerator1
    Position Sampling=::NewEnumerator2
    Skeletal Mesh=$System.Skeletal Mesh
    Socket Sampling Mode=::NewEnumerator5
   ) {module=SkeletalMeshLocation002}
   &VortexForce(
    UseOriginOffset=false
    Vortex Force Amount=&Multiply_Float(A=500, B=1.5)
    Vortex Origin:NiagaraPosition=$Output.SkeletalMeshLocation001.SampledPositionWithoutOffset
    Vortex Origin Offset:Vector3f=&RandomRangeVector(Minimum=<0 0 0>, Maximum=<0 0 0>)
   )
   &ScaleVelocity(Velocity Scale=&Multiply_VectorByFloat(Vector=<1 1 1>, Float=1))
   &LineAttractionForce(
    Attraction Strength=200
    Line End:Vector3f=&AddVectorToPosition(Position=<0 0 500>, Vector=$Output.SkeletalMeshLocation001.SampledPositionWithoutOffset)
    Line Start:Vector3f=&AddVectorToPosition(Position=<0 0 500>, Vector=$Output.SkeletalMeshLocation001.SampledPositionWithoutOffset)
    Use Attraction Falloff=false
   )
   &LinearForce(Coordinate Space:ENiagaraCoordinateSpace=::World, Force=<0 0 9000>)
   &SpringForce(
    " Particle Equilibrium Position":NiagaraPosition=$Particles.Initial.ShapeLocation.ShapePosition
    Attachment Point Velocity:Vector3f=<0 0 500>
    Desired Minimum Separation Distance=1
    Force Strength=0
    Spring Tightness=7
   ) {enabled=false}
   /SetVariables(Transient.PhysicsForce:Vector3f=&Multiply_Vector(
     A=$Transient.PhysicsForce
     B=&Lerp_Vector(A=<40 40 0.85>, B=<80 80 1.5>, Alpha=$System.InterpolateOverTime.InterpolateOverTime_Float.Moving Average)
    )) {isSetParameters=true, module=SetVariables_#7}
   &Drag(Drag=50, Use Linear Drag=true, Use Rotational Drag=true)
   &CurlNoiseForce(Bias Noise=false, Noise Strength=0, Pan Noise=false)
   &SolveForcesAndVelocity(Clamp Velocity=false, Limit Acceleration=false)
   /SetVariables(Particles.Position=$Particles.Position) {isSetParameters=true, module=SetVariables_#4}
   &ParticleState()
   &ScaleSpriteSize(
    Uniform Curve Scale=0.7
    Uniform Curve Sprite Scale=DataInterface() {propertyValues {Curve {keys [
        {interpMode=RCIM_Cubic, tangentMode=RCTM_User, time=0, value=0}
        {interpMode=RCIM_Cubic, tangentMode=RCTM_User, time=0.3, value=1}
        {arriveTangent=-1.4286, interpMode=RCIM_Cubic, leaveTangent=-1.4286, time=0.7, value=1}
        {tangentMode=RCTM_User, time=1, value=0}
       ]}}}
   )
   &ScaleColor(
    Scale Alpha=&FloatFromCurve(FloatCurve=DataInterface() {propertyValues {Curve {keys [
         {interpMode=RCIM_Cubic, tangentMode=RCTM_User, time=0, value=0}
         {interpMode=RCIM_Cubic, tangentMode=RCTM_User, time=0.5, value=1}
         {tangentMode=RCTM_User, time=1, value=0}
        ]}}})
    ScaleA=true
    ScaleRGB=true
   )
   /SetVariables(Particles.Glow03_Color=&MakeLinearColorFromVectorAndFloat(
     "Vector (RGB)"=&MakeVectorFromLinearColorRGB(Color=&MultiplyLinearColor(Linear Color=<rgba 0 1 0.2735 0.2>, Scale Factor=&MakeVector4_FromFloat(Float=1)))
     "Float (Alpha)"=&Multiply_Float(A=&FN_FrontEndFloatModifier(Gameplay=0.2, Front End=0.1), B=1)
    )) {isSetParameters=true}
   /SetVariables(Particles.Glow03_Size=&Multiply_Vector2D_ByFloat(
     Float=&FloatFromCurve(
      FloatCurve=DataInterface() {propertyValues {Curve {keys [
          <at 0 0 RCIM_Cubic>
          <at 0.2 1.3 RCIM_Cubic>
          {arriveTangent=-4.5374, interpMode=RCIM_Cubic, leaveTangent=-4.5374, tangentMode=RCTM_User, time=1, value=0}
         ]}}}
      Scale Curve=13
     )
     Vector2D=<0.6 1.2>
    )) {isSetParameters=true}
  ]
  rendererClasses [&NiagaraSpriteRendererProperties]
  renderers [
   {rendererClass=&NiagaraSpriteRendererProperties}
   {rendererClass=&NiagaraSpriteRendererProperties, rendererIndex=1}
  ]
 }
 TopSmoke {
  emitterUpdateScript [
   &EmitterState()
   &SpawnRate(
    Spawn Probability=&MakeCustomFloatFromBool(Bool=&SetBoolByFloatComparison(A=$User.ParaGlideLeanAlpha, B=0), True Float=1, False Float=0)
    SpawnRate=10
    Use Spawn Probability=true
   )
   /SetVariables(Emitter.ColorDes=-2) {isSetParameters=true, module=SetVariables_#5}
  ]
  particleSpawnScript [
   &InitializeParticle(
    Lifetime=1.5
    Position Offset=&MakeVector(X=0, Y=0, Z=0)
    Sprite Rotation Mode=::NewEnumerator2
    Sprite Size Mode=::NewEnumerator1
    Uniform Sprite Size Max=120
    Uniform Sprite Size Min=100
    UsePositionOffset=false
    Write Material Random=true
   )
   &ShapeLocation(
    Handle Radius=0
    Large Radius=30
    MaskSpawnGroup=false
    Offset=&MakeVector(X=0, Y=0, Z=&RandomRangeFloat(Minimum=290, Maximum=300))
    Offset Mode=::NewEnumerator0
    Override Seed=false
    Shape Origin=&VectorFromFloat(Value=0)
    Shape Primitive=::NewEnumerator3
    UseOwnerScale=true
   )
   /SetVariables(Particles.SpriteSize=&Multiply_Vector2D_ByFloat(
     Float=&FloatFromCurve(
      FloatCurve=DataInterface() {propertyValues {Curve {keys [{interpMode=RCIM_Cubic, tangentMode=RCTM_Break, time=0, value=1}, <at 1 0.3>]}}}
      CurveIndex=&Divide_Float(
       A=&Subtract_Float(A=&MakeFloatFromVector(Vector=$Particles.ShapeLocation.ShapePosition, Channel=::NewEnumerator2), B=100)
       B=170
      )
     )
     Vector2D=$Particles.SpriteSize
    )) {enabled=false, isSetParameters=true, module=SetVariables_#6}
  ]
  particleUpdateScript [
   &SkeletalMeshLocation(
    Bone / Socket Index=5
    Bone Sampling Mode=::NewEnumerator4
    MaskSpawnGroup=false
    Orientation Sampling=::NewEnumerator1
    Position Sampling=::NewEnumerator1
    Sampled Position Offset=&AddVectorToPosition(
     Position=$Particles.Initial.ShapeLocation.ShapePosition
     Vector=&RandomRangeVector(Minimum=<0 0 0>, Maximum=<0 0 30>)
    )
    Skeletal Mesh=$System.Skeletal Mesh
   )
   &ParticleState()
   &SpriteRotationRate(Rotation Rate=&RandomRangeFloat(Minimum=-150, Maximum=150))
   &ScaleSpriteSize(
    Uniform Curve Scale=&FN_FrontEndFloatModifier(Gameplay=1, Front End=0.7)
    Uniform Curve Sprite Scale=DataInterface() {propertyValues {Curve {keys [
        {arriveTangent=-0.321, interpMode=RCIM_Cubic, leaveTangent=-0.321, tangentMode=RCTM_User, time=0, value=0.7}
        {arriveTangent=-0.2921, tangentMode=RCTM_User, time=1, value=1.5}
       ]}}}
   )
   &GravityForce(Gravity=<0 0 -100>)
   &PointForce(
    Force Origin=&AddVectorToPosition(Position=$Output.SkeletalMeshLocation.SampledPositionWithoutOffset, Vector=<0 0 300>)
    Force Strength=200
    Module.Use Falloff=false
    Module.UseLinearForce=false
   )
   &Drag(Drag=10, Use Linear Drag=true, Use Rotational Drag=true)
   &SolveForcesAndVelocity(Clamp Velocity=false, Limit Acceleration=false)
   &ScaleColor(
    Scale Alpha=&FloatFromCurve(
     FloatCurve=DataInterface() {propertyValues {Curve {keys [
         {arriveTangent=1.5124, interpMode=RCIM_Cubic, leaveTangent=1.5124, tangentMode=RCTM_User, time=0, value=0}
         {arriveTangent=-3.5353, interpMode=RCIM_Cubic, leaveTangent=-3.5353, tangentMode=RCTM_User, time=0.3, value=1}
         {arriveTangent=-1.3108, tangentMode=RCTM_User, time=1, value=0}
        ]}}}
     Scale Curve=&Multiply_Float(
      A=2
      B=&ClampFloat(Float=&Lerp_Float(A=0, B=1, Alpha=$System.InterpolateOverTime.InterpolateOverTime_Float.Moving Average))
     )
    )
    Scale RGB=&MakeVectorFromLinearColorRGB(Color=&ColorFromCurve(ColorCurve=DataInterface() {propertyValues {
        AlphaCurve {keys [<at 0 1 RCIM_Cubic>,<at 1 0 RCIM_Cubic>]}
        BlueCurve {keys [<at 0.0954 0.0042 RCIM_Cubic>,<at 0.6039 0.0884>,<at 0.9434 0.0032 RCIM_Cubic>]}
        GreenCurve {keys [<at 0.0954 0.0208 RCIM_Cubic>,<at 0.6039 0.2344>,<at 0.9434 0.1 RCIM_Cubic>]}
        RedCurve {keys [<at 0.0954 0.0009 RCIM_Cubic>,<at 0.6039 0.0155>,<at 0.9434 0.0003 RCIM_Cubic>]}
       }}))
    ScaleA=true
    ScaleRGB=true
   )
   &ScaleColor(
    Scale Alpha=&FloatFromCurve(
     FloatCurve=DataInterface() {propertyValues {Curve {keys [
         {interpMode=RCIM_Cubic, tangentMode=RCTM_User, time=0, value=0}
         {interpMode=RCIM_Cubic, tangentMode=RCTM_User, time=0.4, value=1}
         {tangentMode=RCTM_User, time=1, value=0}
        ]}}}
     Scale Curve=1.5
    )
    Scale RGB=&MakeVectorFromLinearColorRGB(Color=&ColorFromCurve(ColorCurve=DataInterface() {propertyValues {
        AlphaCurve {keys [<at 0 1 RCIM_Cubic>,<at 1 0 RCIM_Cubic>]}
        BlueCurve {keys [<at 0.2197 0.0006 RCIM_Cubic>,<at 1 0.1345 RCIM_Cubic>]}
        GreenCurve {keys [<at 0.2197 0.02 RCIM_Cubic>,<at 1 2 RCIM_Cubic>]}
        RedCurve {keys [<at 0.2197 0.0001 RCIM_Cubic>,<at 1 0.4829 RCIM_Cubic>]}
       }}))
    ScaleA=true
    ScaleRGB=true
   ) {enabled=false, module=ScaleColor001}
  ]
  rendererClasses [&NiagaraSpriteRendererProperties]
  renderers [{rendererClass=&NiagaraSpriteRendererProperties}]
 }
 Tornado {
  emitterUpdateScript [
   &EmitterState(
    Inactive Response=::NewEnumerator1
    Life Cycle Mode=::NewEnumerator1
    Loop Behavior=::NewEnumerator1
    Loop Duration Mode=::NewEnumerator1
    UseLoopDelay=false
   )
   &SpawnBurst_Instantaneous(Use Spawn Probability=false, UseLoopCountLimit=false)
   /SetVariables(Emitter.MainAlpha:NiagaraFloat=&FN_FrontEndFloatModifier(Gameplay=5, Front End=4)) {isSetParameters=true}
  ]
  particleSpawnScript [
   &InitializeParticle(
    Mesh Scale=&Multiply_VectorByFloat(Vector=<1 1 1.1>, Float=1.8)
    Mesh Scale Mode=::NewEnumerator4
    Position Offset=<0 0 130>
    Sprite Size=<300 500>
    Sprite Size Mode=::NewEnumerator4
    UsePositionOffset=false
    Write Material Random=true
   )
   &CameraOffset(Camera Offset Amount=-50)
   &InitialMeshOrientation(
    Mesh Orientation Mode=::NewEnumerator2
    Rotation=&FN_FrontEndVector3Modifier(Gameplay=<0 0 0>, Front End=<0 0 0>)
    Use Rotation Vector=true
   )
  ]
  particleUpdateScript [
   &SkeletalMeshLocation(
    Bone / Socket Index=5
    Bone Sampling Mode=::NewEnumerator4
    MaskSpawnGroup=false
    Sampled Position Offset=&AddVectorToPosition(
     Position=$System.FX offset
     Vector=&MakeVector(
      X=0
      Y=0
      Z=&ClampFloat(
       Float=&Lerp_Float(A=&FN_FrontEndFloatModifier(Gameplay=-40, Front End=-20), B=-20, Alpha=$User.ParaGlideLeanAlpha)
       Min=-40
       Max=-20
      )
     )
    )
    Skeletal Mesh=$System.Skeletal Mesh
   )
   &ParticleState(Kill Particles When Lifetime Has Elapsed=false)
   &ScaleMeshSize(Scale Factor=&Multiply_VectorByFloat(
     Vector=&Multiply_Vector(
      A=&MakeVector(
       X=&ClampFloat(Float=&Lerp_Float(A=0.9, B=1, Alpha=$System.InterpolateOverTime.InterpolateOverTime_Float.Moving Average))
       Y=&ClampFloat(Float=&Lerp_Float(A=0.9, B=1, Alpha=$System.InterpolateOverTime.InterpolateOverTime_Float.Moving Average))
       Z=&ClampFloat(Float=&Lerp_Float(
         A=&FN_FrontEndFloatModifier(Gameplay=0.6, Front End=0.8)
         B=1
         Alpha=$System.InterpolateOverTime.InterpolateOverTime_Float.Moving Average
        ))
      )
      B=$System.SizeScale
     )
     Float=&FN_FrontEndFloatModifier(Gameplay=1, Front End=0.9)
    ))
   &DynamicMaterialParameters(
    Index 0 Param 1=2
    Index 0 Param 2=100
    Index 0 Param 4=&ClampFloat(Float=&Lerp_Float(A=0, B=0, Alpha=$System.InterpolateOverTime.InterpolateOverTime_Float.Moving Average))
    Module.Index0Float0Write=true
    Module.Index0Float1Write=true
    Module.Index0Float2Write=true
    Module.Index0Float3Write=true
   )
   &FN_Sprite_Attributes(
    Color=&MakeLinearColorFromVectorAndFloat(
     "Vector (RGB)"=&Multiply_Vector(A=&MakeVectorFromLinearColorRGB(Color=<rgba 0 0.3 0.01 0.5>), B=<1 1 1>)
     "Float (Alpha)"=&Multiply_Float(A=0.5, B=1)
    )
    Create Color Attribute=true
    Create Position Attribute=true
    Position:NiagaraPosition=&TransformPosition(
     Position=&AddVectorToPosition(Position=$Particles.Position, Vector=<0 0 70>)
     Source Space=::Local
     Destination Space=::Simulation
    )
    Write to Module Namespace=true
   )
   &ScaleColor(
    Scale Alpha=&Multiply_Float(
     A=&FN_FrontEndFloatModifier(Gameplay=1.5, Front End=1)
     B=&ClampFloat(Float=&Lerp_Float(
       A=&FN_FrontEndFloatModifier(Gameplay=0.7, Front End=0.8)
       B=&FN_FrontEndFloatModifier(Gameplay=1, Front End=0.7)
       Alpha=$System.InterpolateOverTime.InterpolateOverTime_Float.Moving Average
      ))
    )
    ScaleA=true
    ScaleRGB=true
   )
   &InterpolateOverTime(Rate Of Change=15, Target Vector Value:NiagaraPosition=$Particles.Position, Variable Type=::NewEnumerator2)
   /SetVariables(Particles.InterpolaterDY:Vector4f=&MakeFloat4FromVectorAndFloat("Vector (RGB)"=$StackContext.InterpolateOverTime.InterpolateOverTime_V3_.Moving Average, "Float (Alpha)"=0.7)) {isSetParameters=true}
  ]
  rendererClasses [&NiagaraMeshRendererProperties]
  renderers [
   {rendererClass=&NiagaraMeshRendererProperties, rendererIndex=0}
   {rendererClass=&NiagaraMeshRendererProperties, rendererIndex=1}
   {rendererClass=&NiagaraMeshRendererProperties, rendererIndex=2}
   {rendererClass=&NiagaraMeshRendererProperties, rendererIndex=3}
  ]
 }
}
summary {
 systemName=NS_Contrail_SprocketPoppyBoat
 emitters [
  {emitterName=Tornado, rendererClasses [&NiagaraMeshRendererProperties]}
  {emitterName=SkullMeshes, rendererClasses [&NiagaraMeshRendererProperties]}
  {emitterName=Souce, rendererClasses [&NiagaraSpriteRendererProperties]}
  {emitterName=GhostTrail, rendererClasses [&NiagaraRibbonRendererProperties]}
  {emitterName=SkullTrail, rendererClasses [&NiagaraRibbonRendererProperties]}
  {emitterName=BottomAirflow, rendererClasses [&NiagaraMeshRendererProperties]}
  {emitterName=BottomSmoke, rendererClasses [&NiagaraSpriteRendererProperties]}
  {emitterName=TopSmoke, rendererClasses [&NiagaraSpriteRendererProperties]}
  {emitterName=Sparkles, rendererClasses [&NiagaraSpriteRendererProperties]}
  {emitterName=BodySkull, rendererClasses [&NiagaraMeshRendererProperties]}
  {emitterName=BodySkullTrail, rendererClasses [&NiagaraRibbonRendererProperties]}
  {emitterName=Glow, rendererClasses [&NiagaraSpriteRendererProperties, &NiagaraLightRendererProperties]}
 ]
 userVariables [
  User.bIsFrontEnd {type=NiagaraBool}
  User.bIsFrontEndPreview {type=NiagaraBool}
  User.ParaGlideLeanAlpha {type=NiagaraFloat, defaultValue=1, description=""}
  User.VariantSwitch {type=NiagaraFloat, defaultValue=0, description=""}
 ]
}
